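# Audio level sequence for shot tdm0100 of GrillMoon Countdown, layering sound and audio-related events over the sequence.
# A master audio track plays one section late in the timeline at default volume and pitch.
# Event trigger tracks on the music and ambience components, both children of the Day 1 SequenceRunner binding, fire events whose targets are not shown.
# The SequenceRunner and mask timer bindings carry no tracks of their own.
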LevelSequence /GrillMoon_Countdown/Sequence/tdm/tdm0100/GrillMoon_Countdown_tdm0100_aud
bindings [
 BP_GrillMoon_POI_MusicComponent {
  id=#1
  parent_id=#2
  possessed_class=BP_GrillMoon_POI_MusicComponent_C
  tracks [{
    type=MovieSceneEventTrack
    sections [{
      type=MovieSceneEventTriggerSection
      channels [None {
        type=MovieSceneScriptingEventChannel
        keys [[3627.42, "<Struct 'MovieSceneEvent' (0x85024314ce0) {}>"], [4021.1575, "<Struct 'MovieSceneEvent' (0x85024311cc0) {}>"]]
       }]
     }]
   }]
 }
 GrillMoon_Countdown_BPmaskTimer {id=#3, possessed_class=GrillMoon_Countdown_BPmaskTimer_C, tracks []}
 BP_GrillMoon_POI_AmbienceComponent {
  id=#4
  parent_id=#2
  possessed_class=BP_GrillMoon_POI_AmbienceComponent_C
  tracks [{
    type=MovieSceneEventTrack
    sections [{
      type=MovieSceneEventTriggerSection
      channels [None {type=MovieSceneScriptingEventChannel, const ["<Struct 'MovieSceneEvent' (0x8502431f000) {}>", 1]}]
     }]
   }]
 }
 BP_GrillMoon_Day1_SequenceRunner {id=#2, possessed_class=BP_GrillMoon_Day1_SequenceRunner_C, tracks []}
]
display_rate {denominator=1, numerator=60}
tick_resolution {denominator=1, numerator=24000}
tracks [{
  type=MovieSceneAudioTrack
  display_name=Audio
  sections [{
    type=MovieSceneAudioSection
    end_frame=4199
    start_frame=3652
    channels [
     Volume {type=MovieSceneScriptingFloatChannel, default=1}
     Pitch {type=MovieSceneScriptingFloatChannel, default=1}
     AttachActor {type=MovieSceneScriptingActorReferenceChannel}
    ]
   }]
 }]
